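# Editor utility widget for HLOD prep that lists static mesh vertex cost across level actors.
# FindHLODCost filters out hidden, editor-only and HLOD-excluded content, then multiplies each mesh's instance count by last-LOD vertices.
# Rows shade green to red by cost, clicks sync the Content Browser, and embedded Python exports CSV to Saved/HLODPrep.
# Both gather buttons pass OnlySelected=false, despite one's "Selected Actors" label; FindLights is never called.
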EditorUtilityWidgetBlueprint /Game/NeverCook/EditorUtilities/PerfTools/HLODPrep/EUW_HLODPrep
graphs [
 EventGraph {nodes [
   "On Pressed (btnSpawnLight)"() {is=title K2Node_ComponentBoundEvent}
   > FindHLODCost!(execute, self, OnlySelected:bool=false) -> (then)
   n1 = "On Item Left Clicked (TA_Generic01_List)"!() -> (OutputDelegate, then, isShiftDown:bool, isCtrlDown:bool, data:Object Reference, item:User Widget Object Reference) {is=title K2Node_ComponentBoundEvent}
   > n2 = Cast To TA_GenericListItemData!(execute, Object=@n1.data) -> (then, CastFailed, AsTA Generic List Item Data:TA Generic List Item Data Object Reference) {is=title K2Node_DynamicCast}
   > SyncBrowserToObjects(AssetPaths=Get strings!(self:TA Generic List Item Data Object Reference=@n2) -> (strings:Array of Strings))
   "On Clicked (btnCSV)"() {is=title K2Node_ComponentBoundEvent}
   > ExportCSV()
   "On Clicked (btnAll)"() {is=title K2Node_ComponentBoundEvent}
   > FindHLODCost(OnlySelected=false)
  ]}
 ExportCSV {nodes [
   ExportCSV() {is=title K2Node_FunctionEntry}
   > n1 = getListData!(execute, self:TA Generic 01 List Object Reference=Get TA_Generic01_List() -> (TA_Generic01_List:TA Generic 01 List Object Reference)) -> (then, data:Array of Object References)
   n2 = "integer > integer"(A=Length(TargetArray:Array of Object References=@n1))
   > PrintString(InString="To String (Boolean)"(InBool=@n2))
   > Branch(Condition=@n2)
   > n3 = Sequence!(execute) -> (then_0, then_1, then_2:Exec)
   > n4 = getListData(self=Get TA_Generic01_List())
   > n5 = For Each Loop(Array:Array of Object References=@n4) -> (Array Element:Object Reference)
   > n6 = Cast To TA_GenericListItemData(Object=@n5)
   n7 = Get integers!(self:TA Generic List Item Data Object Reference=@n6) -> (integers:Array of Integers)
   n8 = "To Text (Integer)"(Value="Get (a copy)"(Array:Array of Integers=@n7, Dimension 1=1) -> (Output:int), bUseGrouping=false)
   n9 = "To Text (Integer)"(Value="Get (a copy)"(Array=@n7, Dimension 1=2), bUseGrouping=false)
   n10 = "To Text (Integer)"(Value="Get (a copy)"(Array=@n7), bUseGrouping=false)
   n11 = "Get (a copy)"(Array:Array of Strings=Get strings(self=@n6)) -> (Output:String)
   > Add(
    TargetArray=Get l_rows() -> (l_rows:Array of Strings)
    NewItem="To String (Text)"(InText=Format Text!(
      Format="LOCTEXT(\"{mesh},{count},{lastLodVerts},{totalverts}\")"
      mesh:String=@n11
      count:Text=@n10
      lastLodVerts:Text=@n9
      totalverts:Text=@n8
     ) -> (Result))
   ) {is=title K2Node_CallArrayFunction}
   Set l_headers(execute=@n3.then_1, l_headers:String="Mesh Name,Count,Last LOD Verts,TotalVerts") -> (Output_Get:String)
   PrintString(execute=@n3.then_2, InString=Hello)
   > Execute Python Script!(
    execute
    PythonScript="import unreal\r\nimport csv\r\nimport os\r\nfrom pathlib import Path\r\nimport subprocess\r\n\r\ndef run(in_headers, in_rows):\r\n    file_path = unreal.Paths.project_saved_dir() + \"HLODPrep/\"\r\n    file_path =  unreal.Paths.convert_relative_path_to_full(file_path)\r\n    file_path =  os.path.realpath(file_path)\r\n    print(\"file path!!!!! \" + file_path)\r\n    final_rows = []\r\n    headers = in_headers.split(\",\")\r\n\r\n    for row in in_rows:\r\n        row_values = row.split(\",\")\r\n        row_dict = {}\r\n        for i in range(len(headers)):\r\n            row_dict[headers[i]] = row_values[i]\r\n        final_rows.append(row_dict)\r\n    \r\n    Path(file_path).mkdir(parents=True, exist_ok=True)\r\n\r\n    file_name = file_path + \"\\HLODPrepExport.csv\"\r\n    print(\"file_name!!!!! \" + file_name)\r\n    with open(file_name, mode='w+', newline='') as csv_file:\r\n        writer = csv.DictWriter(csv_file, fieldnames = headers, extrasaction='ignore')\r\n        writer.writeheader()\t\t\r\n        for each in final_rows:\r\n            writer.writerow(each)\r\n            \r\n    os.startfile(file_path)\r\n\r\nrun(in_headers, in_rows)"
    in_headers:String=Get l_headers() -> (l_headers:String)
    in_rows:Array of Strings=Get l_rows()
   ) -> (then, ReturnValue) {inputs [in_headers,in_rows]}
   MakeLiteralString(Value=",")
  ]}
 FilterActor {nodes [
   n1 = FilterActor!() -> (then, Actor:Actor Object Reference) {is=title K2Node_FunctionEntry}
   > Set l_actor(l_actor:Actor Object Reference=@n1) -> (Output_Get:Actor Object Reference)
   n2 = NOT Boolean(A=Get bEnableAutoLODGeneration!(self:Actor Object Reference=Get l_actor() -> (l_actor:Actor Object Reference)) -> (bEnableAutoLODGeneration:bool))
   n3 = IsTemporarilyHiddenInEditor!(self:Actor Object Reference=@n1, bIncludeParent:bool=false) -> (ReturnValue:bool)
   > Return Node!(execute, Valid:bool=NOT Boolean(A=OR Boolean(A=@n3, B=@n2)))
  ]}
 FilterComp {nodes [
   n1 = FilterComp!() -> (then, comp:Actor Component Object Reference) {is=title K2Node_FunctionEntry}
   > n2 = Cast To StaticMeshComponent(Object=@n1)
   n3 = Is Excluded From HLOD Level!(self:Primitive Component Object Reference=@n2, HLODLevel:EHLODLevelExclusion Enum=HLOD1) -> (ReturnValue:bool)
   n4 = Is Excluded From HLOD Level(self=@n2, HLODLevel=HLOD0)
   n5 = NOT Boolean(A=Get bEnableAutoLODGeneration(self:Primitive Component Object Reference=@n2))
   n6 = NOT Boolean(A=IsValid(Object=@n1))
   n7 = Get bIsEditorOnly!(self:Actor Component Object Reference=@n1) -> (bIsEditorOnly:bool)
   > Return Node(Valid=NOT Boolean(A=OR Boolean!(A=@n7, B=@n6, C:bool=@n4, D:bool=@n3, E:bool=@n5) -> (ReturnValue)))
  ]}
 FindHLODCost {nodes [
   n1 = FindHLODCost!() -> (then, OnlySelected:bool=~) {is=title K2Node_FunctionEntry}
   > n2 = Sequence()
   > n3 = Branch(Condition=@n1)
   n4 = Get EditorActorSubsystem()
   > n5 = GetSelectedLevelActors(self=@n4)
   > n6 = Set l_searchActors(l_searchActors:Array of Actor Object References=@n5) -> (Output_Get:Array of Actor Object References)
   n7 = For Each Loop(Exec=[@n21.then, @n6.then], Array:Array of Actor Object References=Get l_searchActors() -> (l_searchActors:Array of Actor Object References)) -> (Array Element:Actor Object Reference)
   > n8 = FilterActor!(execute, self, Actor:Actor Object Reference=@n7) -> (then, Valid:bool) {is=title K2Node_CallFunction}
   > Branch(Condition=@n8)
   > n9 = For Each Loop(Array:Array of Static Mesh Component Object References=Get Components By Class!(self:Actor Object Reference=@n7, ComponentClass:Actor Component Class Reference=&StaticMeshComponent) -> (ReturnValue:Array of Static Mesh Component Object References)) -> (Array Element:Static Mesh Component Object Reference)
   > n10 = FilterComp!(execute, self, comp:Actor Component Object Reference=@n9) -> (then, Valid:bool) {is=title K2Node_CallFunction}
   > Branch(Condition=@n10)
   > n11 = Set l_mesh(l_mesh:Static Mesh Object Reference=Get StaticMesh(self=@n9)) -> (Output_Get:Static Mesh Object Reference)
   > Is Valid(InputObject=@n11)
   n12 = Cast To InstancedStaticMeshComponent!(Object=@n9) -> (AsInstanced Static Mesh Component:Instanced Static Mesh Component Object Reference, bSuccess:bool=~) {is=title K2Node_DynamicCast}
   n13 = Find(
    TargetMap:Map of Static Mesh Object References to Integers=Get l_meshCountMap() -> (l_meshCountMap:Map of Static Mesh Object References to Integers)
    Key:"Static Mesh Object Reference (by ref)"=Get l_mesh() -> (l_mesh:Static Mesh Object Reference)
   ) -> (Value:int)
   n14 = GetInstanceCount!(self:Instanced Static Mesh Component Object Reference=@n12) -> (ReturnValue:int=0)
   > n15 = Set l_new(l_new:int=SelectInt(A=int + int(A=@n13, B=@n14), B=int + int(A=@n13), bPickA=@n12.bSuccess)) -> (Output_Get:int=0)
   > Add!(
    execute
    TargetMap:Map of Static Mesh Object References to Integers=Get l_meshCountMap()
    Key:"Static Mesh Object Reference (by ref)"=Get l_mesh()
    Value:"Integer (by ref)"=@n15
   ) -> (then) {is=title K2Node_CallFunction}
   n16 = Get StaticMeshEditorSubsystem()
   n17 = Get l_mesh()
   > n18 = GetLodCount!(execute, self:Static Mesh Editor Subsystem Object Reference=@n16, StaticMesh:Static Mesh Object Reference=@n17) -> (then, ReturnValue:int=0)
   n19 = int * int(
    A=GetNumberVerts!(self:Static Mesh Editor Subsystem Object Reference=@n16, StaticMesh:Static Mesh Object Reference=@n17, LODIndex:int=int - int(A=@n18)) -> (ReturnValue:int=0)
    B=Get l_new() -> (l_new:int=0)
   )
   > Set l_max(l_max:int="Max (Integer)"(A=@n19, B=Get l_max() -> (l_max:int=0))) -> (Output_Get:int=0)
   > Add(TargetMap=Get l_meshVertMap() -> (l_meshVertMap:Map of Static Mesh Object References to Integers), Key=Get l_mesh(), Value=@n19)
   n20 = GetAllLevelActors(execute=@n3.else, self=@n4)
   > n21 = Set l_searchActors(l_searchActors=@n20)
   clearItems(execute=@n2.then_1, self:TA Generic 01 List Object Reference=Get TA_Generic01_List())
   n22 = Get l_meshCountMap()
   > n23 = Values!(execute, TargetMap:Map of Static Mesh Object References to Integers=@n22) -> (then, Values:Array of Integers)
   > n24 = Keys!(execute, TargetMap:Map of Static Mesh Object References to Integers=@n22) -> (then, Keys:Array of Static Mesh Object References)
   > n25 = For Each Loop(Array:Array of Static Mesh Object References=@n24) -> (Array Element:Static Mesh Object Reference)
   > Set l_mesh(l_mesh=@n25)
   > Set l_vertCount(l_vertCount:int=Find(TargetMap=Get l_meshVertMap(), Key=Get l_mesh())) -> (Output_Get:int=0)
   n26 = Get StaticMeshEditorSubsystem()
   n27 = Get l_mesh()
   > n28 = GetLodCount(self=@n26, StaticMesh=@n27) -> (ReturnValue=0)
   > Set l_lastLODVertCount(l_lastLODVertCount:int=GetNumberVerts(self=@n26, StaticMesh=@n27, LODIndex=int - int(A=@n28)) -> (ReturnValue=0)) -> (Output_Get:int=0)
   n29 = Break Asset Data!(AssetData=CreateAssetData(InAsset=@n25, bAllowBlueprintClass=false)) -> (PackageName, AssetName)
   n30 = "To Text (Integer)"(Value=Get l_vertCount() -> (l_vertCount:int=0), bUseGrouping=false, MinimumIntegralDigits=5)
   n31 = "Get (a ref)"(Array=@n23, Dimension 1=@n25.Array Index)
   n32 = "To Text (Integer)"(Value=@n31, bUseGrouping=false, MinimumIntegralDigits=3)
   > n33 = addItemGeneric!(
    execute
    self:TA Generic 01 List Object Reference=Get TA_Generic01_List()
    displayString:String="To String (Text)"(InText=Format Text!(Format="LOCTEXT(\"{Verts} - {Count} - {Mesh}\")", Verts:Text=@n30, Count:Text=@n32, Mesh:Name=@n29.AssetName) -> (Result))
   ) -> (then, widget:User Widget Object Reference, data:Object Reference)
   > n34 = Cast To TA_GenericListItemData(Object=@n33.data)
   > Set strings!(
    execute
    strings:Array of Strings=Make Array("[0]":String="To String (Name)"(InName=@n29)) -> (Array:Array of Strings)
    self:TA Generic List Item Data Object Reference=@n34
   ) -> (then, Output_Get:Array of Strings)
   > Set integers!(
    execute
    integers:Array of Integers=Make Array!("[0]":int=@n31, "[1]":int=Get l_vertCount() -> (l_vertCount=0), "[2]":int=Get l_lastLODVertCount() -> (l_lastLODVertCount:int=0)) -> (Array:Array of Integers)
    self:TA Generic List Item Data Object Reference=@n34
   ) -> (then, Output_Get:Array of Integers)
   > n35 = Cast To TA_Generic01_ListItem!(execute, Object=@n33) -> (then, CastFailed, AsTA Generic 01 List Item:TA Generic 01 List Item Object Reference) {is=title K2Node_DynamicCast}
   n36 = "Lerp (LinearColor)"(
    A=(R=0.2281,G=1,B=0,A=1)
    B=(R=1,G=0,B=0,A=1)
    Alpha=float / float(A:int=Get l_vertCount() -> (l_vertCount=0), B:int=Get l_max() -> (l_max=0))
   )
   > Set buttonColor_Default!(
    execute
    buttonColor_Default:Linear Color Structure=LinearColor * LinearColor(A:Linear Color Structure=@n36, B:Linear Color Structure=(R=0.599,G=0.599,B=0.599,A=1)) -> (ReturnValue:Linear Color Structure) {is=title K2Node_PromotableOperator}
    self:TA Base Editor Button Widget Object Reference=@n35
   ) -> (then, Output_Get:Linear Color Structure)
   > Set buttonColor_Selected!(execute, buttonColor_Selected:Linear Color Structure=@n36, self:TA Base Editor Button Widget Object Reference=@n35) -> (then, Output_Get:Linear Color Structure)
   > Set labelColor_Default!(execute, labelColor_Default:Linear Color Structure=@n36, self:TA Base Editor Button Widget Object Reference=@n35) -> (then, Output_Get:Linear Color Structure)
   > validateVisuals!(execute, self:TA Base Editor Button Widget Object Reference=@n35) -> (then, " ":bool)
   n37 = Get l_vertCount() -> (l_vertCount=0)
   n38 = Get l_lastLODVertCount() -> (l_lastLODVertCount=0)
   n39 = Get l_max() -> (l_max=0)
   > SetToolTipText!(
    execute
    self:Widget Object Reference=@n35
    InToolTipText:"Text (by ref)"=Format Text!(
     Format="LOCTEXT(\"Mesh Verts - {verts}\\r\\nMesh Count - {count}\\r\\nTotal Verts - {total}\\r\\nMax Verts - {max}\")"
     verts:Text="To Text (Integer)"(Value=@n38, bUseGrouping=false)
     count:Text="To Text (Integer)"(Value=@n31, bUseGrouping=false)
     total:Text="To Text (Integer)"(Value=@n37, bUseGrouping=false)
     max:Text="To Text (Integer)"(Value=@n39, bUseGrouping=false)
    ) -> (Result)
   ) -> (then)
  ]}
 FindLights {nodes [
   FindLights() {is=title K2Node_FunctionEntry}
   n1 = Get EditorActorSubsystem()
   > n2 = GetAllLevelActorsComponents!(execute, self:Editor Actor Subsystem Object Reference=@n1) -> (then, ReturnValue:Array of Actor Component Object References)
   > n3 = Filter by Class!(
    execute
    TargetArray:Array of Object References=@n2
    ObjectClass:Object Class Reference=/Script/Engine.LightComponent
    FilterType:EEditorScriptingFilterType Enum
   ) -> (then, ReturnValue:Array of Light Component Object References)
   > n4 = For Each Loop(Array:Array of Light Component Object References=@n3) -> (Array Element:Light Component Object Reference)
   n5 = To Array!(execute=@n4.Completed, A:Set of Actor Object References=Get l_actors() -> (l_actors:Set of Actor Object References)) -> (then, Result:Array of Actor Object References)
   > SetSelectedLevelActors(self=@n1, ActorsToSelect=@n5)
   n6 = Cast To PointLightComponent!(execute=@n4.LoopBody, Object=@n4) -> (then, CastFailed, AsPoint Light Component:Point Light Component Object Reference) {is=title K2Node_DynamicCast}
   > Branch(Condition="Not Equal (Enum)"!(
     A:EDetailMode Enum=Get DetailMode!(self:Scene Component Object Reference=@n4) -> (DetailMode:EDetailMode Enum=DM_Low)
     B:EDetailMode Enum=DM_Epic
    ) -> (ReturnValue:bool))
   > Branch(Condition=Get bUseInverseSquaredFalloff!(self:Point Light Component Object Reference=@n6) -> (bUseInverseSquaredFalloff:bool))
   > Add!(execute, TargetSet:Set of Actor Object References=Get l_actors(), NewItem:"Actor Object Reference (by ref)"=GetOwner(self=@n4)) -> (then)
  ]}
 dfsa {nodes [dfsa() {is=title K2Node_FunctionEntry}]}
]
variables [
 Actors to Select {type="array<Actor>"}
 NewVar {type=Object, default=/Game/Athena/Helios/Maps/Landmarks/SW_activation/LM_ImperialRoadblock1}
 Package Names {type="array<string>", default []}
 world {type=Object, default=null}
]
widgets {
 info {rootWidgetClass=&VerticalBox, widgetCount=8}
 widgets [
  VerticalBox_0 {widgetClassPath=&VerticalBox, widget=VerticalBox_0}
  btnSpawnLight {
   widgetClassPath=&Button
   parent=VerticalBox_0
   widget=btnSpawnLight
   properties {widgetStyle {
     clickedSlateSound {resourceObject=null}
     disabled {
      bIsDynamicallyLoaded=false
      drawAs=RoundedBox
      imageType=NoImage
      mirroring=NoMirror
      resourceName=None
      resourceObject=null
      tiling=NoTile
      imageSize=<0 0>
      margin=<ltrb 0 0 0 0>
      outlineSettings {
       bUseBrushTransparency=false
       roundingType=FixedRadius
       width=1
       color {colorUseRule=UseColor_ColorTable, specifiedColor=<rgba 0 0 0 0>}
       cornerRadii=<4 4 4 4>
      }
      tintColor {colorUseRule=UseColor_Specified, specifiedColor=<rgba 0.0395 0.0395 0.0395 1>}
      uVRegion {bIsValid=false, max=<0 0>, min=<0 0>}
     }
     disabledForeground {colorUseRule=UseColor_ColorTable, specifiedColor=<rgba 0 0 0 0>}
     hovered {
      bIsDynamicallyLoaded=false
      drawAs=RoundedBox
      imageType=NoImage
      mirroring=NoMirror
      resourceName=None
      resourceObject=null
      tiling=NoTile
      imageSize=<32 32>
      margin=<ltrb 0 0 0 0>
      outlineSettings {
       bUseBrushTransparency=false
       roundingType=FixedRadius
       width=1
       color {colorUseRule=UseColor_ColorTable, specifiedColor=<rgba 0 0 0 0>}
       cornerRadii=<4 4 4 4>
      }
      tintColor {colorUseRule=UseColor_Specified, specifiedColor=<rgba 0.0953 0.0953 0.0953 1>}
      uVRegion {bIsValid=false, max=<0 0>, min=<0 0>}
     }
     hoveredForeground {colorUseRule=UseColor_ColorTable, specifiedColor=<rgba 0 0 0 0>}
     hoveredSlateSound {resourceObject=null}
     normal {
      bIsDynamicallyLoaded=false
      drawAs=RoundedBox
      imageType=NoImage
      mirroring=NoMirror
      resourceName=None
      resourceObject=null
      tiling=NoTile
      imageSize=<32 32>
      margin=<ltrb 0 0 0 0>
      outlineSettings {
       bUseBrushTransparency=false
       roundingType=FixedRadius
       width=1
       color {colorUseRule=UseColor_ColorTable, specifiedColor=<rgba 0 0 0 0>}
       cornerRadii=<4 4 4 4>
      }
      tintColor {colorUseRule=UseColor_Specified, specifiedColor=<rgba 0.0395 0.0395 0.0395 1>}
      uVRegion {bIsValid=false, max=<0 0>, min=<0 0>}
     }
     normalForeground {colorUseRule=UseColor_ColorTable, specifiedColor=<rgba 0 0 0 0>}
     normalPadding=<ltrb 12 1.5 12 1.5>
     pressed {
      bIsDynamicallyLoaded=false
      drawAs=RoundedBox
      imageType=NoImage
      mirroring=NoMirror
      resourceName=None
      resourceObject=null
      tiling=NoTile
      imageSize=<32 32>
      margin=<ltrb 0 0 0 0>
      outlineSettings {
       bUseBrushTransparency=false
       roundingType=FixedRadius
       width=1
       color {colorUseRule=UseColor_ColorTable, specifiedColor=<rgba 0 0 0 0>}
       cornerRadii=<4 4 4 4>
      }
      tintColor {colorUseRule=UseColor_Specified, specifiedColor=<rgba 0.0284 0.0284 0.0284 1>}
      uVRegion {bIsValid=false, max=<0 0>, min=<0 0>}
     }
     pressedForeground {colorUseRule=UseColor_ColorTable, specifiedColor=<rgba 0 0 0 0>}
     pressedPadding=<ltrb 12 2.5 12 0.5>
     pressedSlateSound {resourceObject=null}
    }}
   slot {class=VerticalBoxSlot}
  }
  TextBlock_56 {
   widgetClassPath=&TextBlock
   parent=btnSpawnLight
   widget=TextBlock_56
   properties {
    text=Gather Mesh Stats on Selected Actors
    colorAndOpacity {colorUseRule=UseColor_Specified, specifiedColor=<rgba 0.8594 0.8594 0.8594 1>}
    font {
     bForceMonospaced=false
     bMaterialIsStencil=false
     fontMaterial=null
     fontObject=&Roboto
     letterSpacing=0
     monospacedWidth=1
     size=12
     skewAmount=0
     typefaceFontName=Bold
     outlineSettings {
      bApplyOutlineToDropShadows=false
      bMiteredCorners=false
      bSeparateFillAlpha=false
      outlineMaterial=null
      outlineSize=0
      outlineColor=<rgba 0 0 0 1>
     }
    }
   }
   slot {class=ButtonSlot}
  }
  btnAll {
   widgetClassPath=&Button
   parent=VerticalBox_0
   widget=btnAll
   properties {widgetStyle {
     clickedSlateSound {resourceObject=null}
     disabled {
      bIsDynamicallyLoaded=false
      drawAs=RoundedBox
      imageType=NoImage
      mirroring=NoMirror
      resourceName=None
      resourceObject=null
      tiling=NoTile
      imageSize=<0 0>
      margin=<ltrb 0 0 0 0>
      outlineSettings {
       bUseBrushTransparency=false
       roundingType=FixedRadius
       width=1
       color {colorUseRule=UseColor_ColorTable, specifiedColor=<rgba 0 0 0 0>}
       cornerRadii=<4 4 4 4>
      }
      tintColor {colorUseRule=UseColor_Specified, specifiedColor=<rgba 0.0395 0.0395 0.0395 1>}
      uVRegion {bIsValid=false, max=<0 0>, min=<0 0>}
     }
     disabledForeground {colorUseRule=UseColor_ColorTable, specifiedColor=<rgba 0 0 0 0>}
     hovered {
      bIsDynamicallyLoaded=false
      drawAs=RoundedBox
      imageType=NoImage
      mirroring=NoMirror
      resourceName=None
      resourceObject=null
      tiling=NoTile
      imageSize=<32 32>
      margin=<ltrb 0 0 0 0>
      outlineSettings {
       bUseBrushTransparency=false
       roundingType=FixedRadius
       width=1
       color {colorUseRule=UseColor_ColorTable, specifiedColor=<rgba 0 0 0 0>}
       cornerRadii=<4 4 4 4>
      }
      tintColor {colorUseRule=UseColor_Specified, specifiedColor=<rgba 0.0953 0.0953 0.0953 1>}
      uVRegion {bIsValid=false, max=<0 0>, min=<0 0>}
     }
     hoveredForeground {colorUseRule=UseColor_ColorTable, specifiedColor=<rgba 0 0 0 0>}
     hoveredSlateSound {resourceObject=null}
     normal {
      bIsDynamicallyLoaded=false
      drawAs=RoundedBox
      imageType=NoImage
      mirroring=NoMirror
      resourceName=None
      resourceObject=null
      tiling=NoTile
      imageSize=<32 32>
      margin=<ltrb 0 0 0 0>
      outlineSettings {
       bUseBrushTransparency=false
       roundingType=FixedRadius
       width=1
       color {colorUseRule=UseColor_ColorTable, specifiedColor=<rgba 0 0 0 0>}
       cornerRadii=<4 4 4 4>
      }
      tintColor {colorUseRule=UseColor_Specified, specifiedColor=<rgba 0.0395 0.0395 0.0395 1>}
      uVRegion {bIsValid=false, max=<0 0>, min=<0 0>}
     }
     normalForeground {colorUseRule=UseColor_ColorTable, specifiedColor=<rgba 0 0 0 0>}
     normalPadding=<ltrb 12 1.5 12 1.5>
     pressed {
      bIsDynamicallyLoaded=false
      drawAs=RoundedBox
      imageType=NoImage
      mirroring=NoMirror
      resourceName=None
      resourceObject=null
      tiling=NoTile
      imageSize=<32 32>
      margin=<ltrb 0 0 0 0>
      outlineSettings {
       bUseBrushTransparency=false
       roundingType=FixedRadius
       width=1
       color {colorUseRule=UseColor_ColorTable, specifiedColor=<rgba 0 0 0 0>}
       cornerRadii=<4 4 4 4>
      }
      tintColor {colorUseRule=UseColor_Specified, specifiedColor=<rgba 0.0284 0.0284 0.0284 1>}
      uVRegion {bIsValid=false, max=<0 0>, min=<0 0>}
     }
     pressedForeground {colorUseRule=UseColor_ColorTable, specifiedColor=<rgba 0 0 0 0>}
     pressedPadding=<ltrb 12 2.5 12 0.5>
     pressedSlateSound {resourceObject=null}
    }}
   slot {class=VerticalBoxSlot}
  }
  TextBlock_1 {
   widgetClassPath=&TextBlock
   parent=btnAll
   widget=TextBlock_1
   properties {
    text=Gather Mesh Stats on All Actors
    colorAndOpacity {colorUseRule=UseColor_Specified, specifiedColor=<rgba 0.8594 0.8594 0.8594 1>}
    font {
     bForceMonospaced=false
     bMaterialIsStencil=false
     fontMaterial=null
     fontObject=&Roboto
     letterSpacing=0
     monospacedWidth=1
     size=12
     skewAmount=0
     typefaceFontName=Bold
     outlineSettings {
      bApplyOutlineToDropShadows=false
      bMiteredCorners=false
      bSeparateFillAlpha=false
      outlineMaterial=null
      outlineSize=0
      outlineColor=<rgba 0 0 0 1>
     }
    }
   }
   slot {class=ButtonSlot}
  }
  TA_Generic01_List {
   widgetClassPath=/Game/Blueprints/Blutilities/ForTechArt/Widgets/List/TA_Generic01_List
   bIsVariable=true
   parent=VerticalBox_0
   widget=TA_Generic01_List
   properties {allowMultiSelect=false}
   slot {class=VerticalBoxSlot, properties {size {sizeRule=Fill, value=1}}}
  }
  btnCSV {
   widgetClassPath=&Button
   parent=VerticalBox_0
   widget=btnCSV
   properties {widgetStyle {
     clickedSlateSound {resourceObject=null}
     disabled {
      bIsDynamicallyLoaded=false
      drawAs=RoundedBox
      imageType=NoImage
      mirroring=NoMirror
      resourceName=None
      resourceObject=null
      tiling=NoTile
      imageSize=<0 0>
      margin=<ltrb 0 0 0 0>
      outlineSettings {
       bUseBrushTransparency=false
       roundingType=FixedRadius
       width=1
       color {colorUseRule=UseColor_ColorTable, specifiedColor=<rgba 0 0 0 0>}
       cornerRadii=<4 4 4 4>
      }
      tintColor {colorUseRule=UseColor_Specified, specifiedColor=<rgba 0.0395 0.0395 0.0395 1>}
      uVRegion {bIsValid=false, max=<0 0>, min=<0 0>}
     }
     disabledForeground {colorUseRule=UseColor_ColorTable, specifiedColor=<rgba 0 0 0 0>}
     hovered {
      bIsDynamicallyLoaded=false
      drawAs=RoundedBox
      imageType=NoImage
      mirroring=NoMirror
      resourceName=None
      resourceObject=null
      tiling=NoTile
      imageSize=<32 32>
      margin=<ltrb 0 0 0 0>
      outlineSettings {
       bUseBrushTransparency=false
       roundingType=FixedRadius
       width=1
       color {colorUseRule=UseColor_ColorTable, specifiedColor=<rgba 0 0 0 0>}
       cornerRadii=<4 4 4 4>
      }
      tintColor {colorUseRule=UseColor_Specified, specifiedColor=<rgba 0.0953 0.0953 0.0953 1>}
      uVRegion {bIsValid=false, max=<0 0>, min=<0 0>}
     }
     hoveredForeground {colorUseRule=UseColor_ColorTable, specifiedColor=<rgba 0 0 0 0>}
     hoveredSlateSound {resourceObject=null}
     normal {
      bIsDynamicallyLoaded=false
      drawAs=RoundedBox
      imageType=NoImage
      mirroring=NoMirror
      resourceName=None
      resourceObject=null
      tiling=NoTile
      imageSize=<32 32>
      margin=<ltrb 0 0 0 0>
      outlineSettings {
       bUseBrushTransparency=false
       roundingType=FixedRadius
       width=1
       color {colorUseRule=UseColor_ColorTable, specifiedColor=<rgba 0 0 0 0>}
       cornerRadii=<4 4 4 4>
      }
      tintColor {colorUseRule=UseColor_Specified, specifiedColor=<rgba 0.0395 0.0395 0.0395 1>}
      uVRegion {bIsValid=false, max=<0 0>, min=<0 0>}
     }
     normalForeground {colorUseRule=UseColor_ColorTable, specifiedColor=<rgba 0 0 0 0>}
     normalPadding=<ltrb 12 1.5 12 1.5>
     pressed {
      bIsDynamicallyLoaded=false
      drawAs=RoundedBox
      imageType=NoImage
      mirroring=NoMirror
      resourceName=None
      resourceObject=null
      tiling=NoTile
      imageSize=<32 32>
      margin=<ltrb 0 0 0 0>
      outlineSettings {
       bUseBrushTransparency=false
       roundingType=FixedRadius
       width=1
       color {colorUseRule=UseColor_ColorTable, specifiedColor=<rgba 0 0 0 0>}
       cornerRadii=<4 4 4 4>
      }
      tintColor {colorUseRule=UseColor_Specified, specifiedColor=<rgba 0.0284 0.0284 0.0284 1>}
      uVRegion {bIsValid=false, max=<0 0>, min=<0 0>}
     }
     pressedForeground {colorUseRule=UseColor_ColorTable, specifiedColor=<rgba 0 0 0 0>}
     pressedPadding=<ltrb 12 2.5 12 0.5>
     pressedSlateSound {resourceObject=null}
    }}
   slot {class=VerticalBoxSlot}
  }
  TextBlock {
   widgetClassPath=&TextBlock
   parent=btnCSV
   widget=TextBlock
   properties {
    text=Export Data to CSV
    colorAndOpacity {colorUseRule=UseColor_Specified, specifiedColor=<rgba 0.8594 0.8594 0.8594 1>}
    font {
     bForceMonospaced=false
     bMaterialIsStencil=false
     fontMaterial=null
     fontObject=&Roboto
     letterSpacing=0
     monospacedWidth=1
     size=12
     skewAmount=0
     typefaceFontName=Bold
     outlineSettings {
      bApplyOutlineToDropShadows=false
      bMiteredCorners=false
      bSeparateFillAlpha=false
      outlineMaterial=null
      outlineSize=0
      outlineColor=<rgba 0 0 0 1>
     }
    }
   }
   slot {class=ButtonSlot}
  }
 ]
}
